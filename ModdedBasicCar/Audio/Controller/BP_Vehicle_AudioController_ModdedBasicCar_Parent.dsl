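; Parent vehicle audio controller for modded basic cars that adds a trunk rattle sound driven by the car's streaming radio.
; On init it binds to the radio player's source started and stopped events to play or fade out the rattle and bind or unbind envelope-follower updates.
; Amplitude from the analysis submix feeds the rattle's "Bass" parameter.
; Pawn enter and exit events update its "IsLocalPassenger" parameter.
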(event VehicleAudio|EventOnInit
  (|CallParentFunction)
  (Utilities|FlowControl|Sequence
    (:then_0
      (bind _output_get (Variables|Default|SetStreamingRadio (Actor|GetComponentbyClass (AudioController|GetVehicleActor) "/Script/SrirachaRanch.StreamingRadioPlayerComponent")))
      (Utilities|IsValid _output_get
        (:Is%20Valid
          (Radio|BindEventtoOnSourcePlayed _output_get (EventDispatchers|CreateEvent :!event Custom|On%20Radio%20Started))
          (Radio|BindEventtoOnSourceStopped _output_get (EventDispatchers|CreateEvent :!event Custom|On%20Radio%20Stopped)))))))

(event Custom|On%20Radio%20Started
  (params (Component :type Streaming%20Radio%20Player%20Component%20Object%20Reference) (AffectedSource :type Athena%20Radio%20Station%20Structure))
  (Audio|Components|Audio|Play (Variables|Default|GetTrunkRattleAudio) 0.0)
  (AudioAnalysis|AssignOnEnvelopeFollowerUpdated (WorldSubsystems|GetFortAudioAnalysisSubsystem) (EventDispatchers|CreateEvent :!event Custom|OnEnvelopeFollowerUpdated_Event)))

(event Custom|On%20Radio%20Stopped
  (params (Component :type Streaming%20Radio%20Player%20Component%20Object%20Reference) (AffectedSource :type Athena%20Radio%20Station%20Structure))
  (Audio|Components|Audio|FadeOut (Variables|Default|GetTrunkRattleAudio) 1.0 0.0 "Linear" :!class /Script/Engine.AudioComponent)
  (AudioAnalysis|UnbindEventfromOnEnvelopeFollowerUpdated (WorldSubsystems|GetFortAudioAnalysisSubsystem) (EventDispatchers|CreateEvent :!event Custom|OnEnvelopeFollowerUpdated_Event)))

(event Custom|OnEnvelopeFollowerUpdated_Event
  (params (Submix :type Sound%20Submix%20Object%20Reference) (Amplitude :type Float%20%28single-precision%29))
  (if (== Submix (Variables|Default|GetAudioAnalysisSubmix))
    (Audio|Parameter|SetFloatParameter (Variables|Default|GetTrunkRattleAudio) "Bass" Amplitude :!class /Script/AudioExtensions.AudioParameterControllerInterface)))

(event Vehicle|EventPawnEnterVehicle
  (params (Vehicle :type Fort%20Vehicle%20Interface%20Interface%20%28by%20ref%29) (PlayerPawn :type Fort%20Pawn%20Object%20Reference) (SeatIndex :type Integer))
  (|CallParentFunction Vehicle PlayerPawn SeatIndex)
  (Audio|Parameter|SetFloatParameter (Variables|Default|GetTrunkRattleAudio) "IsLocalPassenger" (Math|Conversions|ToFloat(Boolean) (Variables|AudioController|GetIsLocalPawnInVehicle)) :!class /Script/AudioExtensions.AudioParameterControllerInterface))

(event Vehicle|EventPawnExitVehicle
  (params (Vehicle :type Fort%20Vehicle%20Interface%20Interface%20%28by%20ref%29) (PlayerPawn :type Fort%20Pawn%20Object%20Reference) (SeatIndex :type Integer))
  (|CallParentFunction Vehicle PlayerPawn SeatIndex)
  (Audio|Parameter|SetFloatParameter (Variables|Default|GetTrunkRattleAudio) "IsLocalPassenger" (Math|Conversions|ToFloat(Boolean) (Variables|AudioController|GetIsLocalPawnInVehicle)) :!class /Script/AudioExtensions.AudioParameterControllerInterface))

(fn ConstructionScript
  (|CallParentFunction))
